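# First-person fire sound cue for the Dragon's Breath sniper weapon.
# A Mixer layers a random pick of four Dragon's Breath fire waves with a random pick of three borrowed bolt-action close, no-tail shots.
# A Modulator boosts the mix by a fixed volume with no pitch variation.
# Shared first-person weapon attenuation, gunfire concurrency and ranged-fire sound class assets govern playback.
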SoundCue /Game/Sounds/Fort_Weapons/Named_Weapons/Dragons_Breath_Sniper/WPN_Sniper_DragonsBreath_1P
root=Modulator_17
nodes [
 {id=Mixer_0, children [Random_1,Random_2], properties {inputVolume [1,1]}}
 {id=Modulator_17, children [Mixer_0], properties {pitchMax=1, pitchMin=1, volumeMax=1.5, volumeMin=1.5}}
 {id=Random_1, children [WavePlayer_7, WavePlayer_8, WavePlayer_9, WavePlayer_10], properties {weights [1,1,1,1]}}
 {id=Random_2, children [WavePlayer_11, WavePlayer_12, WavePlayer_13], properties {weights [1,1,1]}}
 {
  id=WavePlayer_10
  properties {soundWaveAssetPtr=&/WPN_DragonsBreathSniper_Fire_1P_v2_03}
 }
 {id=WavePlayer_11, properties {soundWaveAssetPtr=/Game/Sounds/Fort_Weapons/SniperRifles/Boltaction_HT/BoltAction_HT_Fire_Close_Notail_01}}
 {id=WavePlayer_12, properties {soundWaveAssetPtr=/Game/Sounds/Fort_Weapons/SniperRifles/Boltaction_HT/BoltAction_HT_Fire_Close_Notail_02}}
 {id=WavePlayer_13, properties {soundWaveAssetPtr=/Game/Sounds/Fort_Weapons/SniperRifles/Boltaction_HT/BoltAction_HT_Fire_Close_Notail_03}}
 {
  id=WavePlayer_7
  properties {soundWaveAssetPtr=&/WPN_DragonsBreathSniper_Fire_1P_v2_04}
 }
 {
  id=WavePlayer_8
  properties {soundWaveAssetPtr=&/WPN_DragonsBreathSniper_Fire_1P_v2_01}
 }
 {
  id=WavePlayer_9
  properties {soundWaveAssetPtr=&/WPN_DragonsBreathSniper_Fire_1P_v2_02}
 }
]
properties {
 attenuationSettings=&Fort_Weapon_Projectile_FirstPerson_Athena_Att
 maxDistance=2097152
 soundClassObject=&Fort_WeaponRanged_Fire_1P_PSM
 volumeMultiplier=1
 attenuationOverrides {distanceAlgorithm=NaturalSound, falloffDistance=4000, attenuationShapeExtents {x=1000}}
 concurrencyOverrides {bLimitToOwner=true, maxCount=3, resolutionRule=StopOldest}
 concurrencySet [&Weapon_Gunfire_FirstPerson_Conc]
}
